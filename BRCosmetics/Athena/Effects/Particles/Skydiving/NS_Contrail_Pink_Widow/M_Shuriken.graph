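# M_Shuriken is a Default Lit material for the shuriken Niagara mesh particles in the Pink Widow skydiving contrail cosmetic.
# It holds four shuriken texture samples: a D texture, a normal map, and S and M textures sampled as masks.
# The graph lists no math nodes or parameters, and does not show which material inputs the samples feed.
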Material /BRCosmetics/Athena/Effects/Particles/Skydiving/NS_Contrail_Pink_Widow/M_Shuriken
graph {
 properties {
  bUsedWithNiagaraMeshParticles=true
  pixelDepthOffsetMode=PDOM_Legacy
  previewMesh=&/SM_Shuriken
  usedShadingModels=Default Lit
 }
 nodes [
  TextureSample() {samplerType=SAMPLERTYPE_Normal, texture=&/T_Shuriken_N}
  TextureSample() {samplerType=SAMPLERTYPE_Masks, texture=&/T_Shuriken_S}
  TextureSample() {texture=&/T_Shuriken_D}
  TextureSample() {samplerType=SAMPLERTYPE_Masks, texture=&/T_Shuriken_M}
 ]
}
