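# Niagara skydiving contrail cosmetic (Pink Widow) that trails magenta effects from two skeletal mesh bone/socket points, split by spawn group.
# Layers mirrored, timer-animated swirl meshes, shuriken meshes, spike/flare meshes, glow sprites, noisy sparks and two attached lights.
# System.Dive blends velocities, drag, swirl scale and spin, and light brightness between rest and dive.
# Meshes orient to owner velocity, with separate front-end preview overrides for orientation, offsets and motion.
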NiagaraSystem /BRCosmetics/Athena/Effects/Particles/Skydiving/NS_Contrail_Pink_Widow/NS_Contrail_Pink_Widow
dependencies {
 usedDataInterfaces [
  &NiagaraDataInterfaceCurve
  &NiagaraDataInterfaceArrayFloat3
  &NiagaraDataInterfaceArrayFloat
  &NiagaraDataInterfaceVectorCurve
  &NiagaraDataInterfaceColorCurve
  &NiagaraDataInterfaceVector2DCurve
 ]
 usedDynamicInputs [&SetBoolByFloatComparison, &MakeCustomFloatFromBool]
 usedModules [
  &SystemState
  &self:SystemSpawnScript.SystemScriptSource.SystemScriptGraph.NiagaraNodeAssignment_0.SetVariables_#1
  &self:SystemSpawnScript.SystemScriptSource.SystemScriptGraph.NiagaraNodeAssignment_1.SetVariables_#2
  &self:SystemSpawnScript.SystemScriptSource.SystemScriptGraph.NiagaraNodeAssignment_4.SetVariables_#3
  &FN_Timer
 ]
}
emitters {
 Light {
  emitterUpdateScript [
   &EmitterState(
    Inactive Response=::NewEnumerator1
    Life Cycle Mode=::NewEnumerator1
    Loop Behavior=::NewEnumerator1
    Loop Duration Mode=::NewEnumerator1
    UseLoopDelay=false
   )
   &SpawnBurst_Instantaneous(Spawn Count=2, Use Spawn Probability=false)
  ]
  particleSpawnScript [
   &FN_SetSpawnGroup(Number of Spawn Groups=2)
   &InitializeParticle(
    Color=&MultiplyLinearColor(Linear Color=<rgba 1 0 1 1>, Scale Factor=<5 5 5 1>)
    UsePositionOffset=false
    Write Material Random=true
   )
  ]
  particleUpdateScript [
   &ParticleState(Kill Particles When Lifetime Has Elapsed=false, Let Infinitely Lived Particles Die When Emitter Deactivates=true)
   &SkeletalMeshLocation(
    Apply Sampled Orientation to Mesh Particles=true
    Bone / Socket Index=$Particles.SpawnGroup
    Bone Sampling Mode=::NewEnumerator4
    Inherit Velocity=false
    MaskSpawnGroup=false
    Sampled Position Offset=<0 0 10>
    Sampled Position Offset Coordinate Space:ENiagaraMeshTransforms=::NewEnumerator2
    Skeletal Mesh=$User.SkeletalMesh
   )
   &ScaleColor(
    Scale Alpha=&FNTimeOfDayScalar(Night Scale=0.5, Morning Scale=0.7, Evening Scale=0.7)
    Scale RGB=&VectorFromFloat(Value=&Lerp_Float(A=0.3, B=0.6, Alpha=$System.Dive))
    ScaleA=true
    ScaleRGB=true
   ) {module=ScaleColor001}
  ]
  rendererClasses [&NiagaraLightRendererProperties]
  renderers [{rendererClass=&NiagaraLightRendererProperties, rendererIndex=0}]
 }
 Shuriken {
  emitterUpdateScript [
   &EmitterState()
   &SpawnRate(SpawnRate=24, Use Spawn Probability=false)
  ]
  particleSpawnScript [
   &FN_SetSpawnGroup(Number of Spawn Groups=2)
   &InitializeParticle(
    Lifetime=&RandomRangeFloat(Minimum=0.3, Maximum=0.6)
    Mesh Scale Mode=::NewEnumerator1
    Mesh Uniform Scale Max=1
    Mesh Uniform Scale Min=0.5
    UsePositionOffset=false
    Write Material Random=true
   )
   &SkeletalMeshLocation(
    Bone / Socket Index=$Particles.SpawnGroup
    Bone Sampling Mode=::NewEnumerator4
    Inherit Velocity=false
    MaskSpawnGroup=false
    Sampled Position Offset=&Lerp_Vector(A=<0 0 20>, B=<1 1 10>, Alpha=$System.Dive)
    Skeletal Mesh=$User.SkeletalMesh
   )
   &CameraOffset(Camera Offset Amount=&RandomRangeFloat(Minimum=10, Maximum=20))
   /Niagara/Modules/Spawn/Location/SphereLocation(MaskSpawnGroup=false, Module.Override Randomness=false, Sphere Radius=25, Surface Only=false)
   &AddVelocity(
    Coordinate Space=::World
    Velocity=&Lerp_Vector(
     A=<0 0 200>
     B=&FN_FrontEndVector3Modifier(Gameplay=<0 0 1000>, Front End=<0 0 0>)
     Alpha=$System.Dive
    )
   )
   &AddVelocity(Velocity=&FN_FrontEndVector3Modifier(
     Gameplay=&Lerp_Vector(A=<0 75 0>, B=<1 200 1>, Alpha=$System.Dive)
     Front End=&Lerp_Vector(A=<0 -10 0>, B=<1 -400 300>, Alpha=$System.Dive)
    )) {module=AddVelocity003}
   &InitialMeshOrientation(Use Rotation Vector=false)
   &AddVelocity(
    Cone Axis=<0 0 1>
    Module.UseConeFalloff=true
    Module.UseVelDistribution=true
    Override Seed=false
    Velocity Mode=::NewEnumerator2
    Velocity Speed=150
   ) {module=AddVelocity001}
  ]
  particleUpdateScript [
   &InheritSourceMovement(Applied Position Scale=<0.5 0.5 1>, ApplyForce=false, ApplyPosition=true, ApplyPositionFalloff=false)
   &ParticleState()
   &Drag(Drag=&Lerp_Float(A=1, B=2, Alpha=$System.Dive), Use Linear Drag=true, Use Rotational Drag=false)
   &CurlNoiseForce(Pan Noise=false)
   &SolveForcesAndVelocity(Clamp Velocity=false, Limit Acceleration=false)
   &UpdateMeshOrientation(Rotation Coordinate Space=::NewEnumerator1, Rotation Rate=0.1, Rotation Vector=<1 1 1>)
   &ScaleMeshSize(Scale Factor=&VectorFromCurve(VectorCurve=DataInterface() {propertyValues {
       XCurve {keys [<at 0 1 RCIM_Cubic>, {arriveTangent=-3.0986, interpMode=RCIM_Cubic, leaveTangent=-3.0986, tangentMode=RCTM_User, time=1, value=0}]}
       YCurve {keys [<at 0 1 RCIM_Cubic>, {arriveTangent=-3.0986, interpMode=RCIM_Cubic, leaveTangent=-3.0986, tangentMode=RCTM_User, time=1, value=0}]}
       ZCurve {keys [<at 0 1 RCIM_Cubic>, {arriveTangent=-3.0986, interpMode=RCIM_Cubic, leaveTangent=-3.0986, tangentMode=RCTM_User, time=1, value=0}]}
      }}))
  ]
  rendererClasses [&NiagaraMeshRendererProperties]
  renderers [{rendererClass=&NiagaraMeshRendererProperties, rendererIndex=0}]
 }
 Sparks {
  emitterUpdateScript [
   &EmitterState()
   &SpawnRate(SpawnRate=12, Use Spawn Probability=false)
   /Game/Effects/Niagara/Modules/Skydiving/FN_Emitter_SkydiveAttributes() {module=Emitter_SkydiveAttributes}
  ]
  particleSpawnScript [
   &FN_SetSpawnGroup(Number of Spawn Groups=2)
   &InitializeParticle(
    Color Maximum=<rgba 0.2 0 0.2 1>
    Color Minimum=<rgba 1 0 1 1>
    Color Mode=::NewEnumerator2
    Lifetime=0.75
    Sprite Size Max=&FN_FrontEndVector2Modifier(Front End:Vector2f=<3 12>, Gameplay:Vector2f=<5.5 20>)
    Sprite Size Min=&FN_FrontEndVector2Modifier(Front End:Vector2f=<2.5 5>, Gameplay:Vector2f=<2.5 10>)
    Sprite Size Mode=::NewEnumerator2
    UsePositionOffset=false
    Write Material Random=true
   ) {module=InitializeParticle001}
   &SkeletalMeshLocation(
    Bone / Socket Index=$Particles.SpawnGroup
    Bone Sampling Mode=::NewEnumerator4
    Inherit Velocity=false
    Interpolate Mesh Sampling=false
    MaskSpawnGroup=false
    Sampled Position Offset=<0 10 0>
    Sampled Position Offset Coordinate Space:ENiagaraMeshTransforms=::NewEnumerator2
    Skeletal Mesh=$User.SkeletalMesh
   )
   &SphereLocation(MaskSpawnGroup=false, Module.Override Randomness=false, Sphere Radius=0, Surface Only=false)
   &AddVelocityInCone(
    Cone Angle=60
    Cone Axis=&Lerp_Vector(
     A=<0 0 1>
     B=&FN_FrontEndVector3Modifier(Gameplay=<0 -0.2 1>, Front End=<0 -1 1>)
     Alpha=$System.Dive
    )
    Module.Use Velocity Falloff On Cone Axis=false
    Velocity Strength=&FN_FrontEndFloatModifier(
     Gameplay=&Lerp_Float(A=400, B=800, Alpha=$System.Dive)
     Front End=&Lerp_Float(A=400, B=800, Alpha=$System.Dive)
    )
   )
  ]
  particleUpdateScript [
   /Game/Effects/Niagara/Modules/Skydiving/FN_DivingAttributes/FN_DivingAttributes_v2(Accumulate Dive=0.1, Accumulate Rest=-0.1, Bone Index=0, SkeletalMesh:NiagaraDataInterfaceSkeletalMesh=$User.SkeletalMesh)
   /SetVariables(Particles.Dive01:NiagaraFloat=&Multiply_Float(A=1, B=$Particles.Dive)) {isSetParameters=true}
   &ParticleState()
   &AddVelocity(Velocity=&Lerp_Vector(A=<0 0 10>, B=<0 -10 20>, Alpha=$System.Dive))
   &VectorNoiseForce(Force Amount=1500, Module.Use Random Noise Time Interval=true, Random Noise Time Interval=0.001)
   &Drag(Drag=&RandomRangeFloat(Minimum=2, Maximum=4), Use Linear Drag=true, Use Rotational Drag=false)
   &GravityForce(Gravity=&Multiply_Vector(A=<0 0 250>, B=&Lerp_Vector(A=<0 0 0>, B=<0 0 1>, Alpha=$Particles.Dive01)))
   &SolveForcesAndVelocity(Clamp Velocity=false, Limit Acceleration=false)
   /SetVariables(Particles.MultColor:LinearColor=&MakeLinearColorFromVectorAndFloat(
     "Vector (RGB)"=&Multiply_Vector(
      A=&MakeVectorFromLinearColorRGB(Color=$Particles.Color)
      B=&MakeVectorFromLinearColorRGB(Color=&ColorFromCurve(ColorCurve=DataInterface() {propertyValues {
          AlphaCurve {keys [<at 0 0 RCIM_Cubic>,<at 0.1151 1>,<at 0.9977 1 RCIM_Cubic>]}
          BlueCurve {keys [<at 0.0017 1>,<at 0.9821 1>]}
          GreenCurve {keys [<at 0.0017 1>,<at 0.9821 1>]}
          RedCurve {keys [<at 0.0017 1>,<at 0.9821 1>]}
         }}))
     )
     "Float (Alpha)"=5
    )) {enabled=false, isSetParameters=true}
   &ScaleColor(
    Scale Alpha=&Multiply_Float(A=&FNTimeOfDayScalar(Night Scale=0.5, Morning Scale=0.7, Evening Scale=0.7), B=1)
    ScaleA=true
    ScaleRGB=true
   )
   &ScaleSpriteSizeBySpeed(Max Scale Factor=<1 3>, Min Scale Factor=<1 1.05>)
   &ScaleSpriteSize(Scale Factor=&Vector2DFromCurve(
     Vector2Curve=DataInterface() {propertyValues {
       XCurve {keys [
         <at 0 0 RCIM_Cubic>
         {arriveTangent=-1.6411, interpMode=RCIM_Cubic, leaveTangent=-1.6411, tangentMode=RCTM_User, time=0.15, value=1}
         <at 1 0>
        ]}
       YCurve {keys [<at 0 0 RCIM_Cubic>,<at 0.15 1 RCIM_Cubic>,<at 1 0>]}
      }}
     ScaleCurve=<1 1>
    ))
  ]
  rendererClasses [&NiagaraSpriteRendererProperties]
  renderers [{rendererClass=&NiagaraSpriteRendererProperties}]
 }
 Spike_and_Flare {
  emitterUpdateScript [
   &EmitterState()
   &SpawnRate(SpawnRate=24, Use Spawn Probability=false)
  ]
  particleSpawnScript [
   &FN_SetSpawnGroup(Number of Spawn Groups=2)
   /Niagara/Modules/Spawn/Initialization/InitializeParticle(
    Color=<rgba 10 0 7.5 1>
    Lifetime=0.4
    Mesh Scale=&Multiply_VectorByFloat(Vector=&UniformRangedVector(Minimum=<0.5 0.5 0.5>, Maximum=<1.5 1 1>), Float=10)
    Module.Write Color=true
    Module.Write Lifetime=true
    Module.Write Mass=true
    Module.Write Material Random=true
    Module.Write Position=true
    Module.Write Scale=true
    Module.Write SpriteRotation=false
    Module.Write SpriteSize=true
    Position=&Add_Vector(A=/Niagara/DynamicInputs/Helpers/SimulationPosition(), B=<0 0 0>)
    Sprite Size=<100 100>
   )
   &DynamicMaterialParameters(
    Index 0 Param 3=&UniformRangedFloat(Minimum=1, Maximum=1.5)
    Index 1 Param 1=0.2
    Index 1 Param 2=1
    Module.Param0WriteEnabled=false
    Module.Param0WriteEnabled1=true
    Module.Param1WriteEnabled=false
    Module.Param1WriteEnabled1=true
    Module.Param2WriteEnabled=true
    Module.Param2WriteEnabled1=true
    Module.Param3WriteEnabled=false
    Module.Param3WriteEnabled1=true
    Write Parameter Index 1=true
   ) {module=DynamicMaterialParameters001}
   &ShapeLocation(
    MaskSpawnGroup=false
    Non Uniform Scale=<1 0.7 1>
    Override Seed=false
    Rotation Axis=<0 1 0>
    Rotation Mode=::NewEnumerator1
    Sphere Radius=1
    UseOwnerScale=true
   )
   &CameraOffset(Camera Offset Amount=&RandomRangeFloat(Minimum=5, Maximum=10))
  ]
  particleUpdateScript [
   &ParticleState()
   &SkeletalMeshLocation(
    Bone / Socket Index=$Particles.SpawnGroup
    Bone Sampling Mode=::NewEnumerator4
    MaskSpawnGroup=false
    Orientation Sampling=::NewEnumerator1
    Skeletal Mesh=$User.SkeletalMesh
   )
   &SolveForcesAndVelocity(Clamp Velocity=false, Limit Acceleration=false)
   &ScaleMeshSize(Scale Factor=&VectorFromCurve(VectorCurve=DataInterface() {propertyValues {
       bOptimizeLUT=false
       XCurve {keys [
         {arriveTangent=1.8635, interpMode=RCIM_Cubic, time=0, value=0}
         {arriveTangent=2, interpMode=RCIM_Cubic, leaveTangent=2, time=0.5, value=0.73}
         <at 1 2 RCIM_Cubic>
        ]}
       YCurve {keys [
         {arriveTangent=0.1613, interpMode=RCIM_Cubic, time=0, value=0.2}
         <at 0.5 1 RCIM_Cubic>
         {interpMode=RCIM_Cubic, leaveTangent=-5.6805, time=1, value=0.2}
        ]}
       ZCurve {keys [<at 0 0.2 RCIM_Cubic>, <at 0.5 1 RCIM_Cubic>, {interpMode=RCIM_Cubic, leaveTangent=-5.4621, time=1, value=0.2}]}
      }}))
   &ScaleColor(
    Scale Mode=::NewEnumerator1
    Scale RGBA=&MakeVector4_FromLinearColor(LinearColor=&ColorFromCurve(ColorCurve=DataInterface() {propertyValues {
        AlphaCurve {keys [<at 0 1 RCIM_Cubic>,<at 1 0 RCIM_Cubic>]}
        BlueCurve {keys [<at 0 1 RCIM_Cubic>,<at 1 0 RCIM_Cubic>]}
        GreenCurve {keys [<at 0 1 RCIM_Cubic>,<at 1 0 RCIM_Cubic>]}
        RedCurve {keys [<at 0 1 RCIM_Cubic>,<at 1 0 RCIM_Cubic>]}
       }}))
    ScaleRGBA=true
   )
   /Niagara/Modules/Update/Orientation/V2/OrientMeshToVector(
    Look At Direction:Vector3f=&NormalizeVector(
     Vector To Normalize=&TransformVector(Vector=&Multiply_Vector(A=$Engine.Owner.Velocity, B=<1 1 1>), Source Space=::World, Destination Space=::Local)
     Safe Normalize=true
     Safe Fallback Threshold=0
     Safe Fallback Vector=<0 0 0.001>
    )
    Override Quat=false
    Reference Axis Vector:Vector3f=<-1 0 0>
    UseOptionalSideAxis=false
    UseOptionalUpAxis=false
   ) {enabled=false}
   /SetVariables(Particles.MeshOrientation=/Niagara/DynamicInputs/TypeConversions/MakeCustomQuaternionFromBool(
     Bool:NiagaraBool=$User.bIsFrontEndPreview
     True Quat:Quat4f=&MakeQuatFromAxisAngle(Axis=&Lerp_Vector(A=<1 1 1>, B=<0.5 0.5 1>, Alpha=$System.Dive), AngleInDegrees=-90)
     False Quat:Quat4f=/Niagara/DynamicInputs/Vectors/OrientMeshToVector(
      Look At Direction=&NormalizeVector(
       Vector To Normalize=&TransformVector(Vector=&Multiply_Vector(A=$Engine.Owner.Velocity, B=<1 1 1>), Source Space=::World, Destination Space=::Local)
       Safe Normalize=true
       Safe Fallback Threshold=0
       Safe Fallback Vector=<0 0 0.001>
      )
      Reference Axis Vector=<-1 0 0>
     )
    )) {isSetParameters=true}
   &DynamicMaterialParameters(
    Index 0 Param 1=&FloatFromCurve(FloatCurve=DataInterface() {propertyValues {Curve {keys [<at 0.1 0.1 RCIM_Cubic>,<at 0.8 1>]}}})
    Index 0 Param 2=2
    Index 0 Param 4=&FloatFromCurve(FloatCurve=DataInterface() {propertyValues {Curve {keys [<at 0 1>,<at 1 3>]}}})
    Module.Param0WriteEnabled=true
    Module.Param1WriteEnabled=true
    Module.Param2WriteEnabled=false
    Module.Param3WriteEnabled=true
   )
  ]
  rendererClasses [&NiagaraMeshRendererProperties]
  renderers [{rendererClass=&NiagaraMeshRendererProperties, rendererIndex=0}]
 }
 Spike_center {
  emitterUpdateScript [
   &EmitterState()
   &SpawnRate(SpawnRate=32, Use Spawn Probability=false)
  ]
  particleSpawnScript [
   &FN_SetSpawnGroup(Number of Spawn Groups=2)
   /Niagara/Modules/Spawn/Initialization/V2/InitializeParticle(
    Color=<rgba 1 1 1 10>
    Lifetime Max=0.5
    Lifetime Min=0.3
    Lifetime Mode=::NewEnumerator1
    Sprite Size Max=<5 50>
    Sprite Size Min=<1 20>
    Sprite Size Mode=::NewEnumerator2
    UsePositionOffset=false
    Write Material Random=true
   )
   &ShapeLocation(
    MaskSpawnGroup=false
    Non Uniform Scale=<1 0.7 1>
    Override Seed=false
    Rotation Axis=<0 1 0>
    Rotation Mode=::NewEnumerator1
    Sphere Radius=1
    UseOwnerScale=true
   )
   &AddVelocity(
    Module.Use Falloff Distance=false
    Origin Offset=&FN_FrontEndVector3Modifier(Gameplay=<0 0 -2>, Front End=&Lerp_Vector(A=<0 0 -2>, B=<1 2 -1>, Alpha=$System.Dive))
    Override Seed=false
    Velocity Mode=::NewEnumerator1
    Velocity Speed=50
   )
   &CameraOffset(Camera Offset Amount=&RandomRangeFloat(Minimum=10, Maximum=20))
   &AddVelocity(
    Module.Use Falloff Distance=false
    Origin Offset=&RotateVectorByQuaternion(Vector To Rotate=<0 2000 0>, Quaternion=$Particles.SkeletalMeshLocation.SampledRotation)
    Override Seed=false
    Velocity Mode=::NewEnumerator1
    Velocity Origin=$Particles.SkeletalMeshLocation.SampledPosition
    Velocity Speed=1
   ) {enabled=false, module=AddVelocity001}
  ]
  particleUpdateScript [
   &ParticleState()
   &SkeletalMeshLocation(
    Bone / Socket Index=$Particles.SpawnGroup
    Bone Sampling Mode=::NewEnumerator4
    MaskSpawnGroup=false
    Orientation Sampling=::NewEnumerator1
    Skeletal Mesh=$User.SkeletalMesh
   )
   &SolveForcesAndVelocity(Clamp Velocity=false, Limit Acceleration=false)
   &ScaleSpriteSize(Scale Factor=&Vector2DFromCurve(
     Vector2Curve=DataInterface() {propertyValues {XCurve {keys [<at 0 0.6>,<at 1 1>]}, YCurve {keys [<at 0 0.03>,<at 1 1.75>]}}}
     ScaleCurve=<1 1>
    ))
   &ScaleColor(
    Scale RGB=&Multiply_Vector(
     A=&MakeVectorFromLinearColorRGB(Color=&ColorFromCurve(ColorCurve=DataInterface() {propertyValues {
         AlphaCurve {keys []}
         BlueCurve {keys [<at 0 1.5 RCIM_Cubic>]}
         GreenCurve {keys [<at 0 0 RCIM_Cubic>]}
         RedCurve {keys [<at 0 2 RCIM_Cubic>]}
        }}))
     B=&VectorFromFloat(Value=&FloatFromCurve(FloatCurve=DataInterface() {propertyValues {Curve {keys [<at 0 0.5>,<at 1 0.25>]}}}))
    )
    ScaleA=true
    ScaleRGB=true
   )
   /Niagara/Modules/Update/Orientation/V2/OrientMeshToVector(
    Look At Direction:Vector3f=&NormalizeVector(
     Vector To Normalize=&TransformVector(Vector=&Multiply_Vector(A=$Engine.Owner.Velocity, B=<1 1 1>), Source Space=::World, Destination Space=::Local)
     Safe Normalize=true
     Safe Fallback Threshold=0
     Safe Fallback Vector=<0 0 0.001>
    )
    Override Quat=false
    Reference Axis Vector:Vector3f=<-1 0 0>
    UseOptionalSideAxis=false
    UseOptionalUpAxis=false
   )
   &DynamicMaterialParameters(
    Index 0 Param 1=1000
    Module.Param0WriteEnabled=true
    Module.Param1WriteEnabled=true
    Module.Param2WriteEnabled=true
    Module.Param3WriteEnabled=true
   )
  ]
  rendererClasses [&NiagaraSpriteRendererProperties]
  renderers [{rendererClass=&NiagaraSpriteRendererProperties}]
 }
 Spike_center001 {
  emitterUpdateScript [
   &EmitterState(Life Cycle Mode=::NewEnumerator1, Loop Behavior=::NewEnumerator1, Loop Duration Mode=::NewEnumerator1, UseLoopDelay=false)
   &SpawnBurst_Instantaneous(Spawn Count=2, Use Spawn Probability=false)
  ]
  particleSpawnScript [
   &FN_SetSpawnGroup(Number of Spawn Groups=2)
   &InitializeParticle(
    Color=<rgba 2 0 1.5 0.5>
    Lifetime Max=0.5
    Lifetime Min=0.3
    Lifetime Mode=::NewEnumerator1
    Sprite Size=<75 40>
    Sprite Size Mode=::NewEnumerator4
    UsePositionOffset=false
    Write Material Random=true
   )
   &CameraOffset(Camera Offset Amount=5)
  ]
  particleUpdateScript [
   &ParticleState(Kill Particles When Lifetime Has Elapsed=false, Let Infinitely Lived Particles Die When Emitter Deactivates=true)
   &SkeletalMeshLocation(
    Bone / Socket Index=$Particles.SpawnGroup
    Bone Sampling Mode=::NewEnumerator4
    MaskSpawnGroup=false
    Orientation Sampling=::NewEnumerator1
    Sampled Position Offset=&FN_FrontEndVector3Modifier(
     Gameplay=<0 0 0>
     Front End=&Lerp_Vector(
      A=&SelectVectorFromArray(Direct Array Index=$Particles.SpawnGroup, Vector Selection Array=DataInterface() {propertyValues {FloatData [<5 5 0>, <-5 -5 0>]}})
      B=&SelectVectorFromArray(Direct Array Index=$Particles.SpawnGroup, Vector Selection Array=DataInterface() {propertyValues {FloatData [<10 0 0>, <-10 0 0>]}})
      Alpha=$System.Dive
     )
    )
    Skeletal Mesh=$User.SkeletalMesh
   )
   &ScaleSpriteSize(Scale Factor=&Vector2DFromCurve(
     Vector2Curve=DataInterface() {propertyValues {XCurve {keys [<at 0 0.6>,<at 1 1>]}, YCurve {keys [<at 0 0.03>,<at 1 1.75>]}}}
     ScaleCurve=<1 1>
    ))
   &ScaleColor(
    Scale Alpha=&FN_FrontEndFloatModifier(Gameplay=1, Front End=0.2)
    Scale RGB=&Multiply_Vector(
     A=&MakeVectorFromLinearColorRGB(Color=<rgba 1 1 1 1>)
     B=&VectorFromFloat(Value=&FloatFromCurve(FloatCurve=DataInterface() {propertyValues {Curve {keys [<at 0 0.5>,<at 1 0.25>]}}}))
    )
    ScaleA=true
    ScaleRGB=true
   )
  ]
  rendererClasses [&NiagaraSpriteRendererProperties]
  renderers [{rendererClass=&NiagaraSpriteRendererProperties}]
 }
 Swirl_Shapes {
  emitterUpdateScript [
   &EmitterState(Inactive Response=::NewEnumerator2, Life Cycle Mode=::NewEnumerator1, Loop Duration=1.5, UseLoopDelay=false)
   &SpawnBurst_Instantaneous(Spawn Count=2, Use Spawn Probability=false, UseLoopCountLimit=false)
   &SpawnRate(SpawnRate=3, Use Spawn Probability=false)
  ]
  particleSpawnScript [
   &FN_SetSpawnGroup(Number of Spawn Groups=2)
   &InitializeParticle(
    Color=<rgba 1 1 1 5>
    Lifetime=&RandomRangeFloat(Minimum=1, Maximum=1.5)
    Mesh Scale=&SelectVectorFromArray(Direct Array Index=$Particles.SpawnGroup, Vector Selection Array=DataInterface() {propertyValues {FloatData [<5 5 8.5>, <-5 -5 8.5>]}})
    Mesh Scale Mode=::NewEnumerator4
    UsePositionOffset=false
    Write Material Random=true
   )
   &SkeletalMeshLocation(
    Bone / Socket Index=$Particles.SpawnGroup
    Bone Sampling Mode=::NewEnumerator4
    Inherit Velocity=false
    Interpolate Mesh Sampling=false
    MaskSpawnGroup=false
    Sampled Position Offset Coordinate Space:ENiagaraMeshTransforms=::NewEnumerator2
    Skeletal Mesh=$User.SkeletalMesh
   )
   &CameraOffset(Camera Offset Amount=5)
   &DynamicMaterialParameters(
    Index 3 Param 3=&SelectFloatFromArray(Direct Array Index=$Particles.SpawnGroup, Float Selection Array=DataInterface() {propertyValues {FloatData [0.333,-0.5]}})
    Module.Param0WriteEnabled3=false
    Module.Param1WriteEnabled3=false
    Module.Param2WriteEnabled3=true
    Module.Param3WriteEnabled3=false
    Write Parameter Index 0=false
    Write Parameter Index 3=true
   )
  ]
  particleUpdateScript [
   &ParticleState()
   &SkeletalMeshLocation(
    Bone / Socket Index=$Particles.SpawnGroup
    Bone Sampling Mode=::NewEnumerator4
    Inherit Velocity=false
    Interpolate Mesh Sampling=false
    MaskSpawnGroup=false
    Sampled Position Offset Coordinate Space:ENiagaraMeshTransforms=::NewEnumerator2
    Skeletal Mesh=$User.SkeletalMesh
   ) {module=SkeletalMeshLocation001}
   &InterpolateOverTime(Rate Of Change=10, Target Vector Value:NiagaraPosition=$Particles.Position, Variable Type=::NewEnumerator2)
   /SetVariables(Particles.DY:Vector4f=&MakeFloat4FromVectorAndFloat("Vector (RGB)"=&Multiply_Vector(A=$Output.InterpolateOverTime.Moving Average, B=<1 1 1>), "Float (Alpha)"=1)) {isSetParameters=true, module=SetVariables_#4}
   &DynamicMaterialParameters(
    Index 0 Param 1=&Multiply_Float(A=&RandomRangeFloat(Minimum=1.2, Maximum=2), B=1)
    Index 0 Param 2=&Multiply_Float(
     A=&RandomRangeFloat(Minimum=1, Maximum=2)
     B=&Lerp_Float(A=1, B=5, Alpha=$System.Dive)
    )
    Index 0 Param 3=&Multiply_Float(
     A=&Multiply_Float(
      A=&RandomRangeFloat(Minimum=1.5, Maximum=2)
      B=&Lerp_Float(A=1, B=1.2, Alpha=$System.Dive)
     )
     B=1
    )
    Index 0 Param 4=&RandomRangeFloat(Minimum=0.6, Maximum=1)
    Index 1 Param 1=2
    Index 1 Param 2=&SelectFloatFromArray(Direct Array Index=$Particles.SpawnGroup, Float Selection Array=DataInterface() {propertyValues {FloatData [1,0]}})
    Index 3 Param 1=&Multiply_Float(
     A=&RandomRangeFloat(Minimum=0.8, Maximum=1)
     B=&Lerp_Float(A=1, B=2, Alpha=$System.Dive)
    )
    Index 3 Param 2=&Multiply_Float(
     A=1
     B=&Lerp_Float(A=0, B=&FN_FrontEndFloatModifier(Gameplay=90, Front End=0), Alpha=$System.Dive)
    )
    Index 3 Param 4=&Add_Float(
     A=&SelectFloatFromArray(Direct Array Index=$Particles.SpawnGroup, Float Selection Array=DataInterface() {propertyValues {FloatData [0.333,-0.5]}})
     B=$System.FN_Timer001.Value
    )
    Module.Param0WriteEnabled=true
    Module.Param0WriteEnabled1=true
    Module.Param0WriteEnabled3=true
    Module.Param1WriteEnabled=true
    Module.Param1WriteEnabled1=true
    Module.Param1WriteEnabled3=true
    Module.Param2WriteEnabled=true
    Module.Param2WriteEnabled1=true
    Module.Param2WriteEnabled3=false
    Module.Param3WriteEnabled=true
    Module.Param3WriteEnabled1=true
    Module.Param3WriteEnabled3=true
    Write Parameter Index 1=true
    Write Parameter Index 3=true
   ) {module=DynamicMaterialParameters002}
   &ScaleMeshSize(Scale Factor=&Multiply_VectorByFloat(
     Vector=&MakeVector(
      X=&SelectFloatFromArray(Direct Array Index=$Particles.SpawnGroup, Float Selection Array=DataInterface() {propertyValues {FloatData [-1,1]}})
      Y=&SelectFloatFromArray(Direct Array Index=$Particles.SpawnGroup, Float Selection Array=DataInterface() {propertyValues {FloatData [-1,-1]}})
      Z=&Lerp_Float(
       A=0.9
       B=&FloatFromCurve(FloatCurve=DataInterface() {propertyValues {Curve {keys [
            {interpMode=RCIM_Cubic, tangentMode=RCTM_User, time=0, value=1}
            {arriveTangent=1.1819, interpMode=RCIM_Cubic, leaveTangent=1.1819, tangentMode=RCTM_User, time=0.9308, value=1.5459}
            <at 0.9974 2 RCIM_Cubic>
           ]}}})
       Alpha=$System.Dive
      )
     )
     Float=&Lerp_Float(
      A=&FloatFromCurve(FloatCurve=DataInterface() {propertyValues {Curve {keys [
           {arriveTangent=-0.9315, interpMode=RCIM_Cubic, leaveTangent=-0.9315, tangentMode=RCTM_User, time=0, value=1}
           {arriveTangent=0.9995, interpMode=RCIM_Cubic, leaveTangent=0.9995, tangentMode=RCTM_User, time=0.2223, value=1}
           {arriveTangent=-0.0405, interpMode=RCIM_Cubic, leaveTangent=-0.0405, tangentMode=RCTM_User, time=1, value=2}
          ]}}})
      B=1.2
      Alpha=$System.Dive
     )
    ))
   &SolveForcesAndVelocity(Clamp Velocity=false, Limit Acceleration=false)
   &ScaleColor(
    Scale Alpha=&FloatFromCurve(FloatCurve=DataInterface() {propertyValues {Curve {keys [
         {arriveTangent=9.0725, interpMode=RCIM_Cubic, leaveTangent=9.0725, tangentMode=RCTM_User, time=0, value=0}
         {arriveTangent=-1.0956, interpMode=RCIM_Cubic, leaveTangent=-1.0956, tangentMode=RCTM_User, time=0.1408, value=0.9847}
         {arriveTangent=-0.7529, interpMode=RCIM_Cubic, leaveTangent=-0.7529, tangentMode=RCTM_User, time=1, value=0}
        ]}}})
    ScaleA=true
    ScaleRGB=true
   )
   /SetVariables(Particles.OrientationMaths:Quat4f=&LerpQuaternion(
     Quaternion A:Quat4f=&MakeQuatFromAxisAngle(
      Axis=&SelectVectorFromArray(Direct Array Index=$Particles.SpawnGroup, Vector Selection Array=DataInterface() {propertyValues {FloatData [<-0.05 0 -1>, <-0.05 0 1>]}})
      AngleInDegrees=-30
     )
     Quaternion B:Quat4f=&MakeQuatFromAxisAngle(
      Axis=&SelectVectorFromArray(Direct Array Index=$Particles.SpawnGroup, Vector Selection Array=DataInterface() {propertyValues {FloatData [<0.5 0.75 1>, <0.5 0.75 -1>]}})
      AngleInDegrees=&SelectFloatFromArray(Direct Array Index=$Particles.SpawnGroup, Float Selection Array=DataInterface() {propertyValues {FloatData [50,-50]}})
     )
     Lerp Factor:NiagaraFloat=$System.Dive
     Enable Spherical Linear Interpolation:NiagaraBool=false
    )) {isSetParameters=true}
   &OrientMeshToVector(
    Look At Direction:Vector3f=&NormalizeVector(
     Vector To Normalize=&TransformVector(Vector=&Multiply_Vector(A=$Engine.Owner.Velocity, B=<1 1 1>), Source Space=::World, Destination Space=::Local)
     Safe Normalize=true
     Safe Fallback Threshold=0
     Safe Fallback Vector=<0 0 0.001>
    )
    Override Quat=false
    Reference Axis Vector:Vector3f=<0 0 -1>
    Side Axis Direction:Vector3f=<0 0 0>
    UseOptionalSideAxis=false
    UseOptionalUpAxis=false
   )
   /SetVariables(Particles.MeshOrientation=&MakeCustomQuaternionFromBool(
     Bool:NiagaraBool=$User.bIsFrontEndPreview
     True Quat:Quat4f=&MultiplyQuaternion(
      Quaternion A=$Output.SkeletalMeshLocation.SampledRotation
      Quaternion B=&LerpQuaternion(
       Quaternion A:Quat4f=&MakeQuatFromAxisAngle(Axis=<1 0 0>, AngleInDegrees=0)
       Quaternion B:Quat4f=&MakeQuatFromAxisAngle(Axis=<1 0 0>, AngleInDegrees=60)
       Lerp Factor:NiagaraFloat=$System.Dive
       Enable Spherical Linear Interpolation:NiagaraBool=false
      )
     )
     False Quat:Quat4f=&LerpQuaternion(
      Quaternion A:Quat4f=$Particles.OrientMeshToVector.PreviousFrameQuat
      Quaternion B:Quat4f=&MultiplyQuaternion(
       Quaternion A=$Particles.OrientationMaths
       Quaternion B=&MakeQuatFromAxisAngle(Axis=<0 0 1>, AngleInDegrees=&Multiply_Float(A=$Particles.NormalizedAge, B=720), NormalizeAxis=true)
      )
      Lerp Factor:NiagaraFloat=$System.Dive
      Enable Spherical Linear Interpolation:NiagaraBool=false
     )
    )) {isSetParameters=true}
  ]
  rendererClasses [&NiagaraMeshRendererProperties]
  renderers [{rendererClass=&NiagaraMeshRendererProperties, rendererIndex=0}]
 }
 Swirl_Shapes001 {
  emitterUpdateScript [
   &EmitterState(Inactive Response=::NewEnumerator2, Life Cycle Mode=::NewEnumerator1, Loop Duration=1.5, UseLoopDelay=false)
   &SpawnBurst_Instantaneous(Spawn Count=2, Use Spawn Probability=false, UseLoopCountLimit=false)
   &SpawnRate(SpawnRate=3, Use Spawn Probability=false)
  ]
  particleSpawnScript [
   &FN_SetSpawnGroup(Number of Spawn Groups=2)
   &InitializeParticle(
    Color=<rgba 1 1 1 5>
    Lifetime=&RandomRangeFloat(Minimum=1, Maximum=1.5)
    Mesh Scale=&SelectVectorFromArray(Direct Array Index=$Particles.SpawnGroup, Vector Selection Array=DataInterface() {propertyValues {FloatData [<-5 5 12>, <5 -5 12>]}})
    Mesh Scale Mode=::NewEnumerator4
    UsePositionOffset=false
    Write Material Random=true
   )
   &SkeletalMeshLocation(
    Bone / Socket Index=$Particles.SpawnGroup
    Bone Sampling Mode=::NewEnumerator4
    Inherit Velocity=false
    Interpolate Mesh Sampling=false
    MaskSpawnGroup=false
    Sampled Position Offset Coordinate Space:ENiagaraMeshTransforms=::NewEnumerator2
    Skeletal Mesh=$User.SkeletalMesh
   )
   &CameraOffset(Camera Offset Amount=5)
   &DynamicMaterialParameters(
    Index 3 Param 3=&SelectFloatFromArray(Direct Array Index=$Particles.SpawnGroup, Float Selection Array=DataInterface() {propertyValues {FloatData [-0.15,-0.7]}})
    Module.Param0WriteEnabled3=false
    Module.Param1WriteEnabled3=false
    Module.Param2WriteEnabled3=true
    Module.Param3WriteEnabled3=false
    Write Parameter Index 0=false
    Write Parameter Index 3=true
   )
  ]
  particleUpdateScript [
   &ParticleState()
   &SkeletalMeshLocation(
    Bone / Socket Index=$Particles.SpawnGroup
    Bone Sampling Mode=::NewEnumerator4
    Inherit Velocity=false
    Interpolate Mesh Sampling=false
    MaskSpawnGroup=false
    Sampled Position Offset Coordinate Space:ENiagaraMeshTransforms=::NewEnumerator2
    Skeletal Mesh=$User.SkeletalMesh
   ) {module=SkeletalMeshLocation001}
   &InterpolateOverTime(Rate Of Change=10, Target Vector Value:NiagaraPosition=$Particles.Position, Variable Type=::NewEnumerator2)
   /SetVariables(Particles.DY:Vector4f=&MakeFloat4FromVectorAndFloat("Vector (RGB)"=&Multiply_Vector(A=$Output.InterpolateOverTime.Moving Average, B=<1 1 1>), "Float (Alpha)"=1)) {isSetParameters=true, module=SetVariables_#4}
   &DynamicMaterialParameters(
    Index 0 Param 1=&Multiply_Float(A=&RandomRangeFloat(Minimum=1.2, Maximum=2), B=1)
    Index 0 Param 2=&Multiply_Float(
     A=&RandomRangeFloat(Minimum=1, Maximum=2)
     B=&Lerp_Float(A=1, B=5, Alpha=$System.Dive)
    )
    Index 0 Param 3=&Multiply_Float(
     A=&Multiply_Float(
      A=&RandomRangeFloat(Minimum=1.5, Maximum=2)
      B=&Lerp_Float(A=1, B=1.2, Alpha=$System.Dive)
     )
     B=1
    )
    Index 0 Param 4=&RandomRangeFloat(Minimum=0.6, Maximum=1)
    Index 1 Param 1=2
    Index 1 Param 2=&SelectFloatFromArray(Direct Array Index=$Particles.SpawnGroup, Float Selection Array=DataInterface() {propertyValues {FloatData [0,1]}})
    Index 3 Param 1=&Multiply_Float(
     A=&RandomRangeFloat(Minimum=0.8, Maximum=1)
     B=&Lerp_Float(A=1, B=2, Alpha=$System.Dive)
    )
    Index 3 Param 2=&Multiply_Float(A=1, B=&Lerp_Float(A=0, B=90, Alpha=$System.Dive))
    Index 3 Param 4=&Add_Float(
     A=&SelectFloatFromArray(Direct Array Index=$Particles.SpawnGroup, Float Selection Array=DataInterface() {propertyValues {FloatData [0.1,-0.66]}})
     B=$System.FN_Timer001.Value
    )
    Module.Param0WriteEnabled=true
    Module.Param0WriteEnabled1=true
    Module.Param0WriteEnabled3=true
    Module.Param1WriteEnabled=true
    Module.Param1WriteEnabled1=true
    Module.Param1WriteEnabled3=true
    Module.Param2WriteEnabled=true
    Module.Param2WriteEnabled1=true
    Module.Param2WriteEnabled3=false
    Module.Param3WriteEnabled=true
    Module.Param3WriteEnabled1=true
    Module.Param3WriteEnabled3=true
    Write Parameter Index 1=true
    Write Parameter Index 3=true
   ) {module=DynamicMaterialParameters002}
   &ScaleMeshSize(Scale Factor=&Multiply_VectorByFloat(
     Vector=&MakeVector(
      X=&SelectFloatFromArray(Direct Array Index=$Particles.SpawnGroup, Float Selection Array=DataInterface() {propertyValues {FloatData [-1,1]}})
      Y=&SelectFloatFromArray(Direct Array Index=$Particles.SpawnGroup, Float Selection Array=DataInterface() {propertyValues {FloatData [-1,-1]}})
      Z=&Lerp_Float(
       A=0.9
       B=&FloatFromCurve(FloatCurve=DataInterface() {propertyValues {Curve {keys [
            {interpMode=RCIM_Cubic, tangentMode=RCTM_User, time=0, value=1}
            {arriveTangent=1.3831, interpMode=RCIM_Cubic, leaveTangent=1.3831, tangentMode=RCTM_User, time=0.9308, value=1.5459}
            <at 0.9974 2 RCIM_Cubic>
           ]}}})
       Alpha=$System.Dive
      )
     )
     Float=&Lerp_Float(
      A=&FloatFromCurve(FloatCurve=DataInterface() {propertyValues {Curve {keys [
           {arriveTangent=0.0194, interpMode=RCIM_Cubic, leaveTangent=0.0194, tangentMode=RCTM_User, time=0, value=1}
           {arriveTangent=0.0509, interpMode=RCIM_Cubic, leaveTangent=0.0509, tangentMode=RCTM_User, time=0.2, value=1}
           {arriveTangent=-0.0405, interpMode=RCIM_Cubic, leaveTangent=-0.0405, tangentMode=RCTM_User, time=1, value=2}
          ]}}})
      B=1.6
      Alpha=$System.Dive
     )
    ))
   &SolveForcesAndVelocity(Clamp Velocity=false, Limit Acceleration=false)
   &ScaleColor(
    Scale Alpha=&FloatFromCurve(FloatCurve=DataInterface() {propertyValues {Curve {keys [
         {arriveTangent=9.0725, interpMode=RCIM_Cubic, leaveTangent=9.0725, tangentMode=RCTM_User, time=0, value=0}
         {arriveTangent=-1.0956, interpMode=RCIM_Cubic, leaveTangent=-1.0956, tangentMode=RCTM_User, time=0.1408, value=0.9847}
         {arriveTangent=-0.7529, interpMode=RCIM_Cubic, leaveTangent=-0.7529, tangentMode=RCTM_User, time=1, value=0}
        ]}}})
    ScaleA=true
    ScaleRGB=true
   )
   /SetVariables(Particles.OrientationMaths:Quat4f=&LerpQuaternion(
     Quaternion A:Quat4f=&MakeQuatFromAxisAngle(
      Axis=&SelectVectorFromArray(Direct Array Index=$Particles.SpawnGroup, Vector Selection Array=DataInterface() {propertyValues {FloatData [<-0.05 0 -1>, <-0.05 0 1>]}})
      AngleInDegrees=-30
     )
     Quaternion B:Quat4f=&MakeQuatFromAxisAngle(
      Axis=&SelectVectorFromArray(Direct Array Index=$Particles.SpawnGroup, Vector Selection Array=DataInterface() {propertyValues {FloatData [<0.5 0.75 1>, <0.5 0.75 -1>]}})
      AngleInDegrees=&SelectFloatFromArray(Direct Array Index=$Particles.SpawnGroup, Float Selection Array=DataInterface() {propertyValues {FloatData [50,-50]}})
     )
     Lerp Factor:NiagaraFloat=$System.Dive
     Enable Spherical Linear Interpolation:NiagaraBool=false
    )) {isSetParameters=true}
   &OrientMeshToVector(
    Look At Direction:Vector3f=&NormalizeVector(
     Vector To Normalize=&TransformVector(Vector=&Multiply_Vector(A=$Engine.Owner.Velocity, B=<1 1 1>), Source Space=::World, Destination Space=::Local)
     Safe Normalize=true
     Safe Fallback Threshold=0
     Safe Fallback Vector=<0 0 0.001>
    )
    Override Quat=false
    Reference Axis Vector:Vector3f=<0 0 -1>
    UseOptionalSideAxis=false
    UseOptionalUpAxis=false
   )
   /SetVariables(Particles.MeshOrientation=&MakeCustomQuaternionFromBool(
     Bool:NiagaraBool=$User.bIsFrontEndPreview
     True Quat:Quat4f=&MultiplyQuaternion(
      Quaternion A=$Output.SkeletalMeshLocation.SampledRotation
      Quaternion B=&LerpQuaternion(
       Quaternion A:Quat4f=&MakeQuatFromAxisAngle(Axis=<1 0 0>, AngleInDegrees=0)
       Quaternion B:Quat4f=&MakeQuatFromAxisAngle(Axis=<1 0 0>, AngleInDegrees=60)
       Lerp Factor:NiagaraFloat=$System.Dive
       Enable Spherical Linear Interpolation:NiagaraBool=false
      )
     )
     False Quat:Quat4f=&LerpQuaternion(
      Quaternion A:Quat4f=$Particles.OrientMeshToVector.PreviousFrameQuat
      Quaternion B:Quat4f=&MultiplyQuaternion(
       Quaternion A=$Particles.OrientationMaths
       Quaternion B=&MakeQuatFromAxisAngle(Axis=<0 0 1>, AngleInDegrees=&Multiply_Float(A=$Particles.NormalizedAge, B=-720), NormalizeAxis=true)
      )
      Lerp Factor:NiagaraFloat=$System.Dive
      Enable Spherical Linear Interpolation:NiagaraBool=false
     )
    )) {isSetParameters=true}
  ]
  rendererClasses [&NiagaraMeshRendererProperties]
  renderers [{rendererClass=&NiagaraMeshRendererProperties, rendererIndex=0}]
 }
}
summary {
 systemName=NS_Contrail_Pink_Widow
 emitters [
  {emitterName=Swirl_Shapes, rendererClasses [&NiagaraMeshRendererProperties]}
  {emitterName=Shuriken, rendererClasses [&NiagaraMeshRendererProperties]}
  {emitterName=Spike_and_Flare, rendererClasses [&NiagaraMeshRendererProperties]}
  {emitterName=Spike_center, rendererClasses [&NiagaraSpriteRendererProperties]}
  {emitterName=Swirl_Shapes001, rendererClasses [&NiagaraMeshRendererProperties]}
  {emitterName=Sparks, rendererClasses [&NiagaraSpriteRendererProperties]}
  {emitterName=Light, rendererClasses [&NiagaraLightRendererProperties]}
  {emitterName=Spike_center001, rendererClasses [&NiagaraSpriteRendererProperties]}
 ]
 userVariables [
  User.bIsFrontEnd {type=NiagaraBool}
  User.bIsFrontEndPreview {type=NiagaraBool}
  User.bIsFullyDeployed {type=NiagaraBool}
  User.ForwardDot {type=NiagaraFloat, defaultValue=0, description=""}
  User.IsSkydiving {type=NiagaraBool}
  User.ParaGlideLeanAlpha {type=NiagaraFloat, defaultValue=0, description=""}
  User.PlayerSpeed {type=NiagaraFloat, defaultValue=0, description=""}
  User.RibbonIDOffset {type=NiagaraInt32}
  User.RightDot {type=NiagaraFloat, defaultValue=0, description=""}
  User.SkeletalMesh {
   type=NiagaraDataInterfaceSkeletalMesh
   defaultValue {
    struct=NiagaraExt_VariableValue_DataInterface
    value {dataInterface=&self:NiagaraDataInterfaceSkeletalMesh_3}
   }
  }
  User.Variant_Color {type=LinearColor, description="", defaultValue {struct=LinearColor, value=<rgba 1 1 1 1>}}
 ]
}
